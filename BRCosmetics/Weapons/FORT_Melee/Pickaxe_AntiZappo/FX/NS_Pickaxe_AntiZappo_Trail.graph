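# Niagara swing trail for the AntiZappo pickaxe, built from two short-lived ribbon emitters using the shared FN anim-trail modules.
# Lightning draws two ribbon renderers, with curves driving material parameters, alpha fade and a bright blue flash that quickly dims, scaled by time of day.
# TrailMotionBlur adds a pale ribbon tinted by a system time-of-day color.
# Its alpha fades over particle life and early emitter age, with separate front-end and gameplay intensities.
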NiagaraSystem /BRCosmetics/Weapons/FORT_Melee/Pickaxe_AntiZappo/FX/NS_Pickaxe_AntiZappo_Trail
dependencies {
 usedDataInterfaces [&NiagaraDataInterfaceColorCurve, &NiagaraDataInterfaceCurve, &NiagaraDataInterfaceSocketReader]
 usedDynamicInputs []
 usedModules [
  &self:SystemSpawnScript.SystemScriptSource.SystemScriptGraph.NiagaraNodeAssignment_1.SetVariables_#1
  &self:SystemSpawnScript.SystemScriptSource.SystemScriptGraph.NiagaraNodeAssignment_0.SetVariables_#2
  &self:SystemSpawnScript.SystemScriptSource.SystemScriptGraph.NiagaraNodeAssignment_2.SetVariables_#3
  &SystemState
 ]
}
emitters {
 Lightning {
  emitterSpawnScript [&FN_RibbonIDOffsetUtility()]
  emitterUpdateScript [
   &EmitterState()
   &FN_SpawnAnimTrail()
  ]
  particleSpawnScript [
   &FN_InitializeAnimTrailBoneIndecies()
   &FN_InitializeAnimTrail(AnimTrailWidth=1.2, Color=<rgba 10 10 10 1>, Lifetime=0.1)
  ]
  particleUpdateScript [
   &SolveForcesAndVelocity(Clamp Velocity=false, Limit Acceleration=false)
   &FN_UpdateAnimTrail() {module=FN_UpdateAnimTrail001}
   /SetVariables(Particles.TrailWidth=1) {isSetParameters=true}
   /SetVariables(Particles.DY2:Vector4f=&MakeVector4(X=1, Y=1, Z=1, W=1)) {isSetParameters=true}
   /SetVariables(Particles.Trail Color02:LinearColor=&MakeLinearColorFromVectorAndFloat(
     "Vector (RGB)"=&MakeVectorFromLinearColorRGB(Color=&MultiplyLinearColor(
       Linear Color=<rgba 0.0288 0.1906 1 0.9>
       Scale Factor=&MakeVector4_FromLinearColor(LinearColor=&ColorFromCurve(ColorCurve=DataInterface() {propertyValues {
           AlphaCurve {keys []}
           BlueCurve {keys [
             <at 0.0045 3>
             {arriveTangent=-8.8662, interpMode=RCIM_Cubic, leaveTangent=-8.8662, time=0.0771, value=1.5}
             <at 0.2542 0.7865>
             <at 0.399 0.25>
            ]}
           GreenCurve {keys [
             <at 0.0045 3>
             {arriveTangent=-8.8662, interpMode=RCIM_Cubic, leaveTangent=-8.8662, time=0.0771, value=1.5}
             <at 0.2542 0.7865>
             <at 0.399 0.25>
            ]}
           RedCurve {keys [
             <at 0.0045 3>
             {arriveTangent=-8.8662, interpMode=RCIM_Cubic, leaveTangent=-8.8662, time=0.0771, value=1.5}
             <at 0.2542 0.7865>
             <at 0.399 0.25>
            ]}
          }}))
      ))
     "Float (Alpha)"=0.5
    )) {isSetParameters=true}
   &DynamicMaterialParameters(
    Index 0 Param 1=10
    Index 0 Param 2=&FloatFromCurve(
     FloatCurve=DataInterface() {propertyValues {Curve {keys [
         {arriveTangent=-3.0186, interpMode=RCIM_Cubic, leaveTangent=-3.0186, tangentMode=RCTM_User, time=0, value=1}
         {interpMode=RCIM_Cubic, tangentMode=RCTM_User, time=1, value=0}
        ]}}}
     Scale Curve=0.8
    )
    Module.Param0WriteEnabled=true
    Module.Param1WriteEnabled=true
    Module.Param2WriteEnabled=false
    Module.Param3WriteEnabled=false
   )
   &ScaleColor(
    Scale Alpha=&FloatFromCurve(
     FloatCurve=DataInterface() {propertyValues {Curve {keys [{arriveTangent=-2.4711, interpMode=RCIM_Cubic, leaveTangent=-2.4711, tangentMode=RCTM_User, time=0, value=1}, <at 1 0>]}}}
     Scale Curve=&Multiply_Float(A=1, B=1)
    )
    Scale RGB=&Multiply_VectorByFloat(
     Vector=&MakeVectorFromLinearColorRGB(Color=&ColorFromCurve(ColorCurve=DataInterface() {propertyValues {
         AlphaCurve {keys []}
         BlueCurve {keys [<at 0 100 RCIM_Cubic>,<at 0.2838 1 RCIM_Cubic>]}
         GreenCurve {keys [
           {arriveTangent=-200.3047, interpMode=RCIM_Cubic, leaveTangent=-200.3047, tangentMode=RCTM_User, time=0, value=16.6667}
           <at 0.2838 0.0833 RCIM_Cubic>
          ]}
         RedCurve {keys [<at 0 0 RCIM_Cubic>,<at 0.2838 0 RCIM_Cubic>]}
        }}))
     Float=&Multiply_Float(A=5, B=&FNTimeOfDayScalar(Night Scale=0.5, Morning Scale=1, Evening Scale=1))
    )
    ScaleA=true
    ScaleRGB=true
   )
  ]
  rendererClasses [&NiagaraRibbonRendererProperties]
  renderers [
   {rendererClass=&NiagaraRibbonRendererProperties}
   {rendererClass=&NiagaraRibbonRendererProperties, rendererIndex=1}
  ]
 }
 TrailMotionBlur {
  emitterSpawnScript [&FN_RibbonIDOffsetUtility()]
  emitterUpdateScript [
   &EmitterState()
   &FN_SpawnAnimTrail(AnimTrailWidthMode=::FromSecond)
  ]
  particleSpawnScript [
   &FN_InitializeAnimTrailBoneIndecies()
   &FN_InitializeAnimTrail(Color=<rgba 0.25 0.24 0.3 1>, Lifetime=0.13)
  ]
  particleUpdateScript [
   &SolveForcesAndVelocity(Clamp Velocity=false, Limit Acceleration=false)
   &FN_UpdateAnimTrail() {module=FN_UpdateAnimTrail001}
   &ScaleColor(
    Scale Alpha=&FloatFromCurve(
     FloatCurve=DataInterface() {propertyValues {Curve {keys [{interpMode=RCIM_Cubic, tangentMode=RCTM_User, time=0, value=1}, {interpMode=RCIM_Cubic, tangentMode=RCTM_User, time=1, value=0}]}}}
     Scale Curve=&FloatFromCurve(
      FloatCurve=DataInterface() {propertyValues {Curve {keys [<at 0.1 1 RCIM_Cubic>,<at 0.25 0 RCIM_Cubic>]}}}
      CurveIndex=$Emitter.Age
      Scale Curve=&FN_FrontEndFloatModifier(
       Gameplay=&FNTimeOfDayScalar(Night Scale=0.4, Morning Scale=0.3, Day Scale=0.5, Evening Scale=0.3)
       Front End=0.25
      )
     )
    )
    Scale RGB=&MakeVectorFromLinearColorRGB(Color=$System.ToDColor)
    ScaleA=true
    ScaleRGB=true
   )
   &ScaleRibbonWidth()
  ]
  rendererClasses [&NiagaraRibbonRendererProperties]
  renderers [{rendererClass=&NiagaraRibbonRendererProperties}]
 }
}
summary {
 systemName=NS_Pickaxe_AntiZappo_Trail
 emitters [
  {emitterName=Lightning, rendererClasses [&NiagaraRibbonRendererProperties]}
  {emitterName=TrailMotionBlur, rendererClasses [&NiagaraRibbonRendererProperties]}
 ]
 userVariables [
  User.bIsFrontEnd {type=NiagaraBool}
  User.bIsFrontEndPreview {type=NiagaraBool}
  User.RibbonIDOffset {type=NiagaraInt32}
 ]
}
